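# Niagara one-shot effect for the CarminePack back bling cosmetic, flashing the gauntlet's gems in a brief snap flare.
# Gem_Flares bursts six sprites, using each burst's spawn group to index arrays of gem colors and local position offsets.
# These heavily brightened flares pop and shrink, while Corona adds a larger blue glow peaking later, driven by dynamic material parameters.
# Both emitters use short lifetimes and camera offsets to pull sprites toward the viewer.
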NiagaraSystem /BRCosmetics/Athena/Items/Gameplay/BackPacks/CarminePack/FX/NS_Jim_Gauntlet_Flare_Snap_Backpack
dependencies {
 usedDataInterfaces [&NiagaraDataInterfaceArrayColor, &NiagaraDataInterfaceArrayFloat3, &NiagaraDataInterfaceCurve]
 usedDynamicInputs []
 usedModules [&SystemState]
}
emitters {
 Corona {
  emitterUpdateScript [
   &EmitterState()
   &SpawnBurst_Instantaneous(Spawn Count=1, Use Spawn Probability=false)
  ]
  particleSpawnScript [
   &InitializeParticle(
    Color=<rgba 0.1344 0.4283 1 1>
    Lifetime=0.3
    Sprite Size Mode=::NewEnumerator3
    Uniform Sprite Size=75
    UsePositionOffset=false
    Write Material Random=true
   )
   &CameraOffset(Camera Offset Amount=35)
  ]
  particleUpdateScript [
   &ParticleState()
   &ScaleColor(
    Scale Alpha=&FloatFromCurve(FloatCurve=DataInterface() {propertyValues {Curve {keys [
         <at 0 0 RCIM_Cubic>
         {arriveTangent=1.4286, interpMode=RCIM_Cubic, leaveTangent=1.4286, time=0.55, value=0.33}
         {arriveTangent=-6.331, interpMode=RCIM_Cubic, leaveTangent=-6.331, tangentMode=RCTM_User, time=0.7, value=1}
         <at 1 0 RCIM_Cubic>
        ]}}})
    ScaleA=true
    ScaleRGB=true
   )
   &DynamicMaterialParameters(
    Index 0 Param 1=&FloatFromCurve(FloatCurve=DataInterface() {propertyValues {Curve {keys [<at 0 1>,<at 0.5 1>,<at 0.7 0.475>,<at 1 0.25>]}}})
    Index 0 Param 2=1
    Index 0 Param 3=0.6
    Index 0 Param 4=1
    Module.Param0WriteEnabled=true
    Module.Param1WriteEnabled=true
    Module.Param2WriteEnabled=true
    Module.Param3WriteEnabled=true
   )
  ]
  rendererClasses [&NiagaraSpriteRendererProperties]
  renderers [{rendererClass=&NiagaraSpriteRendererProperties}]
 }
 Gem_Flares {
  emitterUpdateScript [
   &EmitterState()
   &SpawnBurst_Instantaneous(Spawn Count=1, Use Spawn Probability=false)
   &SpawnBurst_Instantaneous(Spawn Count=1, Spawn Group=1, Use Spawn Probability=false) {module=SpawnBurst_Instantaneous001}
   &SpawnBurst_Instantaneous(Spawn Count=1, Spawn Group=2, Use Spawn Probability=false) {module=SpawnBurst_Instantaneous002}
   &SpawnBurst_Instantaneous(Spawn Count=1, Spawn Group=3, Use Spawn Probability=false) {module=SpawnBurst_Instantaneous003}
   &SpawnBurst_Instantaneous(Spawn Count=1, Spawn Group=4, Use Spawn Probability=false) {module=SpawnBurst_Instantaneous004}
   &SpawnBurst_Instantaneous(Spawn Count=1, Spawn Group=5, Use Spawn Probability=false) {module=SpawnBurst_Instantaneous005}
  ]
  particleSpawnScript [
   &InitializeParticle(
    Color=&SelectLinearColorFromArray(
     Array Sampling Mode=::NewEnumerator1
     Direct Array Index=&SpawnGroup()
     Color Selection Array=DataInterface() {propertyValues {ColorData [
        <rgba 0.0479 1 0.1589 2>
        <rgba 0.0267 0.3123 1 2>
        <rgba 1 0.01 0 2>
        <rgba 0.6603 0.0409 1 2>
        <rgba 1 0.6858 0.0249 2>
        <rgba 1 0.3282 0 2>
       ]}}
     Random Seed=?
    )
    Lifetime=0.3
    Position Offset=&SelectVectorFromArray(
     Direct Array Index=&SpawnGroup()
     Vector Selection Array=DataInterface() {propertyValues {FloatData [<-4.75 -2.75 8.75>, <6.2 -13.3 7.55>, <9.25 -12.5 5.55>, <2.75 -13.5 9.4>, <5 -13 2>, <11.75 -10.5 3.4>]}}
    )
    Position Offset Coordinate Space=::Local
    Sprite Rotation Angle=0.2
    Sprite Rotation Mode=::NewEnumerator1
    Sprite Size Mode=::NewEnumerator3
    Uniform Sprite Size=50
    UsePositionOffset=true
    Write Material Random=true
   )
   &CameraOffset(Camera Offset Amount=10)
  ]
  particleUpdateScript [
   &ParticleState()
   &ScaleColor(
    Scale Alpha=&FloatFromCurve(FloatCurve=DataInterface() {propertyValues {Curve {keys [
         <at 0 0 RCIM_Cubic>
         {arriveTangent=1.5146, interpMode=RCIM_Cubic, leaveTangent=1.5146, tangentMode=RCTM_User, time=0.175, value=0.175}
         {arriveTangent=-2.249, interpMode=RCIM_Cubic, leaveTangent=-2.249, tangentMode=RCTM_User, time=0.25, value=1}
         <at 1 0 RCIM_Cubic>
        ]}}})
    Scale RGB=&VectorFromFloat(Value=25)
    ScaleA=true
    ScaleRGB=true
   )
   &ScaleSpriteSize(Uniform Curve Sprite Scale=DataInterface() {propertyValues {Curve {keys [
        <at 0 0.4 RCIM_Cubic>
        {arriveTangent=0.0116, interpMode=RCIM_Cubic, leaveTangent=0.0116, tangentMode=RCTM_User, time=0.125, value=1}
        {arriveTangent=-0.8118, interpMode=RCIM_Cubic, leaveTangent=-0.8118, tangentMode=RCTM_User, time=0.5, value=0.67}
        <at 1 0.5 RCIM_Cubic>
       ]}}})
  ]
  rendererClasses [&NiagaraSpriteRendererProperties]
  renderers [{rendererClass=&NiagaraSpriteRendererProperties}]
 }
}
summary {
 systemName=NS_Jim_Gauntlet_Flare_Snap_Backpack
 emitters [
  {emitterName=Gem_Flares, rendererClasses [&NiagaraSpriteRendererProperties]}
  {emitterName=Corona, rendererClasses [&NiagaraSpriteRendererProperties]}
 ]
 userVariables []
}
